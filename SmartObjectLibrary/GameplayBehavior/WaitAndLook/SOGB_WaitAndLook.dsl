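; Smart object gameplay behavior that makes an AI bot stand and look along the smart object's forward direction for a random time.
; Init casts and caches the pawn, AI bot controller, WaitAndLook config and owner.
; On success it sets a focal point ahead of the pawn, then ends after a delay drawn from the config's range, clamped to a small minimum.
; Failed init ends the behavior immediately; finishing clears focus.
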(event GameplayBehavior|EventOnTriggeredCharacter
  (params (Avatar :type Character%20Object%20Reference) (Config :type Gameplay%20Behavior%20Config%20Object%20Reference) (SmartObjectOwner :type Actor%20Object%20Reference))
  (bind _output_get (Variables|Default|SetUserAvatar Avatar))
  (bind _successfulinit (CallFunction|Init _output_get Config SmartObjectOwner))
  (if _successfulinit
    (CallFunction|SetFocalPoint)
    (GameplayTasks|TaskWaitDelay 0 (Math|Float|Max(Float) (Math|Random|RandomFloatinRange (Variables|/SmartObjectLibrary/GameplayBehavior/WaitAndLook/SOGB_WaitAndLook_Config.SOGB_WaitAndLook_Config_C|GetWaitDurationMin (Variables|Default|GetWaitAndLookConfig)) (Variables|/SmartObjectLibrary/GameplayBehavior/WaitAndLook/SOGB_WaitAndLook_Config.SOGB_WaitAndLook_Config_C|GetWaitDurationMax (Variables|Default|GetWaitAndLookConfig))) 0.05)
      (:OnFinish
        (GameplayBehavior|EndBehavior (Variables|Default|GetUserAvatar))))
    (else
      (bind _useravatar (Variables|Default|GetUserAvatar)
        (:then
          (GameplayBehavior|EndBehavior _useravatar))
        (:else)))))

(event GameplayBehavior|EventOnFinishedCharacter
  (params (Avatar :type Character%20Object%20Reference) (bWasInterrupted :type Boolean))
  (CallFunction|ResetFocalPoint))

(fn Init
  (params (Avatar :type Character%20Object%20Reference) (BehaviorConfig :type Gameplay%20Behavior%20Config%20Object%20Reference) (SmartObjectOwner :type Actor%20Object%20Reference))
  (Utilities|FlowControl|Sequence
    (:then_0
      (bind _asfort_player_pawn (Utilities|Casting|CastToFortPlayerPawn Avatar)
        (:then
          (bind _output_get (Variables|Default|SetUserPawn _asfort_player_pawn))
          (bind _asfort_athena_aibot_controller (Utilities|Casting|CastToFortAthenaAIBotController (Pawn|GetController _output_get))
            (:then
              (Variables|Default|SetUserController _asfort_athena_aibot_controller))
            (:CastFailed
              _)))
        (:CastFailed
          _))
      (LocalVariables|SetSuccessfullyInitialized false))
    (:then_1
      (bind _assogb_wait_and_look_config (Utilities|Casting|CastToSOGB_WaitAndLook_Config BehaviorConfig)
        (:then
          (Variables|Default|SetWaitAndLookConfig _assogb_wait_and_look_config))
        (:CastFailed
          (LocalVariables|SetSuccessfullyInitialized false))))
    (:then_2
      (Variables|Default|SetSmartObjectOwner SmartObjectOwner))
    (:then_3
      (return (LocalVariables|GetSuccessfullyInitialized)))))

(fn ResetFocalPoint
  (bind _usercontroller (Variables|Default|GetUserController)
    (:then
      (AI|ClearFocus _usercontroller))
    (:else)))

(fn SetFocalPoint
  (if (Math|Boolean|ANDBoolean (Utilities|IsValid (Variables|Default|GetUserPawn) :!class /Script/Engine.KismetSystemLibrary) (Utilities|IsValid (Variables|Default|GetUserController) :!class /Script/Engine.KismetSystemLibrary) (Utilities|IsValid (Variables|Default|GetSmartObjectOwner) :!class /Script/Engine.KismetSystemLibrary))
    (AI|SetFocalPoint (Variables|Default|GetUserController) (+ (Transformation|GetActorLocation (Variables|Default|GetUserPawn)) (* (Transformation|GetActorForwardVector (Variables|Default|GetSmartObjectOwner)) 1e+02)))))
